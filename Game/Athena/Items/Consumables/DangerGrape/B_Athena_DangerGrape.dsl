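; DangerGrape consumable actor that spawns a temporary AI decoy of the owning player.
; On BeginPlay it overrides the spawner's cosmetic component with the player's loadout, then requests a spawn matched by request ID.
; The bot joins the owner's team, copies a supported ranged weapon, gets a no-damage effect and heads to a navmesh point ahead.
; A curve-table timer kills it; gameplay events report spawn and removal, and batched damage feeds an absorb quest.
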(event EventBeginPlay
  (bind _asfort_player_pawn (Utilities|Casting|CastToFortPlayerPawn (Actor|GetOwner)))
  (bind _returnvalue (AI|GetAISpawner))
  (Utilities|IsValid _returnvalue
    (:Is%20Valid
      ;[0.4828 1 0.9201] Bind SetupSpawnedBot to AISpawner::OnPawnSpawned
      (EventDispatchers|BindEventtoOnPawnSpawned _returnvalue (EventDispatchers|CreateEvent :!event Custom|SetupSpawnedBot))
      ;[1 0.3222 0.352] Create AISpawner Component List from a AISpawner class
      (bind _returnvalue_2 (Class|FortAthenaAISpawnerData|CreateComponentListfromClass (select (GameplayTags|HasAnyMatchingGameplayTags _asfort_player_pawn (Variables|Default|GetNo%20Starting%20Weapons%20Tag)) (Variables|Default|GetAISpawnerData_NoStartingWeapons) (Variables|Default|GetAISpawnerData)) self))
      (Variables|Default|SetNPCComponentList _returnvalue_2)
      (Utilities|FlowControl|Sequence
        (:then_0
          ;[1 0.2353 0.7171] Find DangerGrape's AISC cosmetics and override it to use the player's loadout...
          (bind _asbp_aisc_cosmetics_danger_grape (Utilities|Casting|CastToBP_AISC_Cosmetics_DangerGrape (Class|FortAthenaAISpawnerDataComponentList|GetorCreateComponentClassforModification (Variables|Default|GetNPCComponentList) self "/Game/Athena/AI/DangerGrape/BP_AISC_Cosmetics_DangerGrape.BP_AISC_Cosmetics_DangerGrape_C")))
          (bind _output_get (Variables|Default|SetCustomAISCCosmetic _asbp_aisc_cosmetics_danger_grape))
          (Variables|/Game/Athena/AI/DangerGrape/BP_AISC_Cosmetics_DangerGrape.BP_AISC_Cosmetics_DangerGrape_C|SetPlayerPawnLoadout_Deprecated (Variables|/Script/FortniteGame.FortPlayerPawn|GetCosmeticLoadout _asfort_player_pawn) _output_get)
          (bind _outoverrides (Class|FortGameStateAthena|GatherCustomCharacterPartOverridesfromMutator (Fortnite|Athena|GameState|GetGameStateAthena) (Fortnite|GetFortPlayerControllerfromActor _asfort_player_pawn)))
          (Variables|/Game/Athena/AI/DangerGrape/BP_AISC_Cosmetics_DangerGrape.BP_AISC_Cosmetics_DangerGrape_C|SetPlayerCustomCharacterParts _outoverrides (Variables|Default|GetCustomAISCCosmetic))
          (bind _returnvalue_3 (AI|OverrideAISpawnerDataComponent 0 (Variables|Default|GetCustomAISCCosmetic) (Variables|Default|GetNPCComponentList)))
          (Variables|Default|SetNPCComponentList _returnvalue_3))
        (:then_1
          ;[1 0.8927 0.2891] Use AISpawnerData to Spawn AI
          (bind _returnvalue_4 (Class|AthenaAISpawner|RequestSpawn _returnvalue (Variables|Default|GetNPCComponentList) (Math|Transform|MakeTransform (Transformation|GetActorLocation self) (Math|Rotator|MakeRotator (.roll (Transformation|GetActorRotation self)) (.pitch (Transformation|GetActorRotation self)) (- (.yaw (Transformation|GetActorRotation self)) 9e+01))) false -1))
          (Variables|Default|SetSpawnRequestID _returnvalue_4))))))

(event Custom|SetupSpawnedBot
  (params (Pawn :type Pawn%20Object%20Reference) (RequestID :type Integer))
  (bind (_projectedlocation _returnvalue) (AI|Navigation|ProjectPointtoNavigation (+ (Transformation|GetActorLocation (Variables|Default|GetOwnerPawn)) (* (Math|Vector|GetForwardVector (Transformation|GetActorRotation (Variables|Default|GetOwnerPawn))) (ScalableFloat|GetValueatLevel (Variables|Default|GetDistanceFromPlayerToMoveDecoy)))) 0 "/Game/Athena/AI/Phoebe/NavFilter/NavFilter_Phoebe.NavFilter_Phoebe_C" "600.000000,600.000000,600.000000"))
  (if (== RequestID (Variables|Default|GetSpawnRequestID))
    (Variables|Default|SetNPCComponentList 0)
    (Variables|Default|SetCustomAISCCosmetic 0)
    ;[0 0.15 0.02544 0.5] Cache Variables
    (bind _asfort_player_pawn_athena (Utilities|Casting|CastToFortPlayerPawnAthena Pawn))
    (bind _output_get (Variables|Default|SetBotPawn _asfort_player_pawn_athena))
    (bind _asfort_athena_aibot_controller (Utilities|Casting|CastToFortAthenaAIBotController (Pawn|GetController _output_get)))
    (Variables|Default|SetBotController _asfort_athena_aibot_controller)
    (bind _asfort_player_pawn_athena_2 (Utilities|Casting|CastToFortPlayerPawnAthena (Actor|GetOwner)))
    (bind _output_get_2 (Variables|Default|SetOwnerPawn _asfort_player_pawn_athena_2))
    (Variables|Default|SetOwnerController (Pawn|GetController _output_get_2))
    (Class|FortPlayerPawn|SetClientObservedStatsfromAnotherPawn (Variables|Default|GetBotPawn) (Variables|Default|GetOwnerPawn))
    (Class|FortAthenaAIBotController|SwitchTeam (Variables|Default|GetBotController) (Team|GetActorTeam (Variables|Default|GetOwnerController)))
    (Utilities|FlowControl|Sequence
      (:then_0
        ;[1 0.1511 0.2408] Try to give the bot a weapon the player's using
        (bind _success (CallFunction|TryGiveBotWeapon (Weapon|GetPreviousWeapon (Variables|Default|GetOwnerPawn))))
        (if (not _success)
          (bind _array_element (Utilities|Array|ForEachLoopwithBreak (Variables|/Script/FortniteGame.FortPawn|GetCurrentWeaponList (Variables|Default|GetOwnerPawn)) :!label _label)
            (:LoopBody
              (bind _success_2 (CallFunction|TryGiveBotWeapon _array_element))
              (if _success_2
                (goto _label Break))))))
      (:then_1
        ;[0.3016 0.2639 1] Apply a GameEffect that will remove Damage done.
        (GameplayEffects|ApplyGameplayEffectToTarget :self (Ability|GetAbilitySystemComponent (Variables|Default|GetBotPawn)) :GameplayEffectClass "/Game/Athena/Items/Consumables/DangerGrape/GE_DangerGrape_NoDamage.GE_DangerGrape_NoDamage_C" :Target (Ability|GetAbilitySystemComponent (Variables|Default|GetBotPawn)) :Level 0.0))
      (:then_2
        ;[0.4914 1 0.5337] Handle Quest Update For Absorbing Damage
        (bind _asfort_player_controller (Utilities|Casting|CastToFortPlayerController (Variables|Default|GetOwnerController)))
        (bind (_questisvalid _questcompleted _objectivecompleted) (Default|AthenaCheckQuestandObjectiveCompletion _asfort_player_controller (Variables|Quest|GetQuest%20Item) (Variables|Quest|GetQuest%20Backend%20Name)))
        (if (and _questisvalid (not _objectivecompleted))
          (EventDispatchers|BindEventtoOnDamaged (Variables|Default|GetBotPawn) (delegate Custom|PawnDamaged) :!class /Script/FortniteGame.FortPawn)))
      (:then_3
        ;[0.5094 1 0.9701] Handle Despawning after X seconds
        (bind _outxy (Fortnite|EvaluateCurveTableRow (Variables|Default|GetDecoyLifetimeRow)))
        (Utilities|Time|SetTimerbyEvent (delegate Custom|RemoveDecoy) _outxy)
        (Actor|SetLifeSpan (+ _outxy 2.0))
        ;[0.9514 1 0.3058] Setup Leashing
        (AI|Components|Blackboard|SetValueasVector (AI|GetBlackboard (Variables|Default|GetBotController)) (Variables|BlackBoardKeys|GetBB_Key_Destination) (select _returnvalue _projectedlocation (+ (* (Math|Vector|GetForwardVector (Transformation|GetActorRotation (Variables|Default|GetOwnerPawn))) 2.5e+02) (Transformation|GetActorLocation (Variables|Default|GetOwnerPawn)))))
        ;[1 1 1] Setup Ability on Player to keep track of amount of Decoys Spawned
        (Ability|SendGameplayEventtoActor (Variables|Default|GetOwnerPawn) "(TagName=\"Abilities.Generic.DangerGrape.Spawned\")" (Utilities|Struct|MakeGameplayEventData :OptionalObject (Variables|Default|GetBotPawn) :ContextHandle "()" :InstigatorTags "(GameplayTags=)" :TargetTags "(GameplayTags=)" :EventMagnitude 0.0 :TargetData "()")))
      (:then_4
        (EventDispatchers|BindEventtoOnDeathEffects (Variables|Default|GetBotPawn) (delegate Custom|PawnDied)))
      (:then_5
        ;[0.9146 0.4428 1] Unbind SetupSpawnedBot from AISpawner::OnPawnSpawned
        (bind _returnvalue_2 (AI|GetAISpawner))
        (Utilities|IsValid _returnvalue_2
          (:Is%20Valid
            (EventDispatchers|UnbindEventfromOnPawnSpawned _returnvalue_2 (EventDispatchers|CreateEvent :!event Custom|SetupSpawnedBot))))))))

(event Custom|PawnDamaged
  (params (DamagedActor :type Actor%20Object%20Reference) (Damage :type Float%20%28single-precision%29) (InstigatedBy :type Controller%20Object%20Reference) (DamageCauser :type Actor%20Object%20Reference) (HitLocation :type Vector) (FHitComponent :type Primitive%20Component%20Object%20Reference) (BoneName :type Name) (Momentum :type Vector))
  ;[0.4914 1 0.5337] Handle Quest Update For Absorbing Damage
  (Variables|Quest|SetDamage (+ Damage (Variables|Quest|GetDamage)))
  (if (not (Team|OnSameTeam DamageCauser (Variables|Default|GetOwnerPawn)))
    (Utilities|FlowControl|RetriggerableDelay 0.2)
    (Quest|SendComplexCustomStatEvent (Quests|GetQuestManager (stale Custom|SetupSpawnedBot _asfort_player_controller) "Invalid") 0 (Variables|Quest|GetT_QuestAbsorbDamage) "(GameplayTags=)" (Math|Float|Truncate (Variables|Quest|GetDamage)))
    (Variables|Quest|SetDamage 0.0)))

(event Custom|UpdateLeashLocation
  ;[0.9514 1 0.3058] Setup Leashing
  (AI|Components|Blackboard|SetValueasVector (AI|GetBlackboard (Variables|Default|GetBotController)) (Variables|BlackBoardKeys|GetBB_Key_LeashLocation) (Transformation|GetActorLocation (Variables|Default|GetOwnerPawn))))

(event Custom|RemoveDecoy
  ;[0.5094 1 0.9701] Handle Despawning after X seconds
  (Utilities|FlowControl|Sequence
    (:then_0
      (Ability|SendGameplayEventtoActor (Variables|Default|GetOwnerPawn) "(TagName=\"Abilities.Generic.DangerGrape.Unspawned\")" (Utilities|Struct|MakeGameplayEventData :OptionalObject (Variables|Default|GetBotPawn) :ContextHandle "()" :InstigatorTags "(GameplayTags=)" :TargetTags "(GameplayTags=)" :EventMagnitude 0.0 :TargetData "()")) ; Update Decoy Count
    )
    (:then_1
      (FortPawn|ForceKill :self (Variables|Default|GetBotPawn)))))

(event Custom|PawnDied
  (params (DeadActor :type Actor%20Object%20Reference) (Damage :type Float%20%28single-precision%29) (InstigatedBy :type Fort%20Pawn%20Object%20Reference) (HitLocation :type Vector) (DamageCauser :type Actor%20Object%20Reference))
  ;[0.5094 1 0.9701] Handle Despawning after X seconds
  (Ability|SendGameplayEventtoActor (Variables|Default|GetOwnerPawn) "(TagName=\"Abilities.Generic.DangerGrape.Unspawned\")" (Utilities|Struct|MakeGameplayEventData :OptionalObject (Variables|Default|GetBotPawn) :ContextHandle "()" :InstigatorTags "(GameplayTags=)" :TargetTags "(GameplayTags=)" :EventMagnitude 0.0 :TargetData "()")) ; Update Decoy Count
)

(fn TryGiveBotWeapon
  (params (Weapon :type Fort%20Weapon%20Object%20Reference))
  ; Check if weapon is valid.
  (Utilities|IsValid Weapon
    (:Is%20Valid
      (bind _returnvalue (AI|IsWeaponSupported Weapon))
      (if (and _returnvalue (Utilities|Enum|Equal(Enum) "WeaponRanged" (Item|GetWeaponDataItemType Weapon)))
        ; Reset inventory with weapon.
        (Class|FortAthenaAIBotController|InitializeInventory (Variables|Default|GetBotController) (Utilities|Array|MakeArray (Utilities|Struct|MakeItemandCount 1 (Data|GetWeaponData Weapon) (Item|GetItemVariantGuid (FortWeapon|GetInventoryItem Weapon)) (Class|FortWeapon|GetWeaponModSlots Weapon))))
        (return true)))))

(fn ConstructionScript)
