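; GA_Athena_DangerGrape is the gameplay ability for the DangerGrape consumable item.
; On activation it runs the parent logic, then applies the GE_Athena_GrantTracker_DangerGrape effect to its owner.
; It overrides ShouldOnlyShowTrajectoryOnUse to return true, so the trajectory shows only during use.
; The end-ability event is a disabled placeholder.
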(event Ability|EventActivateAbility
  (|CallParentFunction)
  (Ability|ApplyGameplayEffectToOwner "/Game/Athena/Items/Consumables/DangerGrape/GE_Athena_GrantTracker_DangerGrape.GE_Athena_GrantTracker_DangerGrape_C"))

(event Ability|EventOnEndAbility
  (params (bWasCancelled :type Boolean))
  (|CallParentFunction bWasCancelled) ; This node is disabled and will not be called.\nDrag off pins to build functionality.
)

(fn ShouldOnlyShowTrajectoryOnUse
  (return true))
