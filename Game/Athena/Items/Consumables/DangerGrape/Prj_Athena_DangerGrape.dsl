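; Thrown Danger Grape projectile that, on stopping or entering water, spawns a configurable actor (noted as a bot spawner) and explodes.
; A hit normal above a Z tolerance stops projectile movement, so floor throws explode instead of bouncing.
; Stop and water events play impact sounds, with water adding a particle effect.
; Each then uses a DoOnce and server check to spawn the yaw-offset actor, explode, hide, disable collision and expire shortly after.
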(event OnEnterWaterEvent(WaterInteractionComponent)
  (params (WaterBody :type Fort%20Water%20Body%20Actor%20Object%20Reference) (WaterInteractionComponent :type Fort%20Water%20Interaction%20Component%20Object%20Reference) (bIsFirstBody :type Boolean))
  (bind (_returnvalue_roll _returnvalue_pitch _returnvalue_yaw) (Math|Rotator|BreakRotator (Transformation|GetActorRotation)))
  (bind _returnvalue (Transformation|GetActorLocation))
  (Audio|PlaySoundatLocation (Variables|Sound|GetWater%20Impact%20Sound) _returnvalue)
  (Effects|SpawnEmitteratLocation "/Game/Athena/Items/Consumables/DangerGrape/P_DangerGrape_Impact_Explosion_Water.P_DangerGrape_Impact_Explosion_Water" _returnvalue)
  ;[1 1 1] Spawn Actor That Spawns Bot on hit and then hide the actor,disable collision, and set lifespan
  (Utilities|FlowControl|DoOnce false
    (:Completed
      (if (Networking|IsServer)
        (Game|SpawnActorfromClass (Variables|Default|GetActorToSpawn) (Math|Transform|MakeTransform (Transformation|GetActorLocation) (Math|Rotator|MakeRotator 0.0 0.0 (+ _returnvalue_yaw 9e+01)) "1.000000,1.000000,1.000000") "AlwaysSpawn" "SelectDefaultAtRuntime" (Game|GetInstigator) (Game|GetInstigator))
        (Delivery|DoExplosion)
        (Rendering|SetActorHiddenInGame true)
        (Collision|SetActorEnableCollision false)
        (Actor|SetLifeSpan 1.0)))))

(event EventOnStop
  (params (Hit :type Hit%20Result%20Structure%20%28by%20ref%29))
  (bind (_returnvalue_roll _returnvalue_pitch _returnvalue_yaw) (Math|Rotator|BreakRotator (Transformation|GetActorRotation)))
  (bind _returnvalue (Transformation|GetActorLocation))
  ;[1 1 1] Spawn Actor That Spawns Bot on hit and then hide the actor,disable collision, and set lifespan
  (|CallParentFunction Hit)
  (Audio|PlaySoundatLocation (Variables|Sound|GetImpact%20Spawn%20Sound) _returnvalue)
  (Utilities|FlowControl|DoOnce false
    (:Completed
      (if (Networking|IsServer)
        (Game|SpawnActorfromClass (Variables|Default|GetActorToSpawn) (Math|Transform|MakeTransform _returnvalue (Math|Rotator|MakeRotator 0.0 0.0 (+ _returnvalue_yaw 9e+01)) "1.000000,1.000000,1.000000") "AlwaysSpawn" "SelectDefaultAtRuntime" (Game|GetInstigator) (Game|GetInstigator))
        (Delivery|DoExplosion)
        (Rendering|SetActorHiddenInGame true)
        (Collision|SetActorEnableCollision false)
        (Actor|SetLifeSpan 1.0)))))

(event EventBeginPlay
  (|CallParentFunction)
  (Networking|CanExecuteCosmeticEvents
    (:True
      (LOD|SetMaxDrawDistance (Variables|Default|GetFuse_Particle) 4e+03))))

(event Collision|EventHit
  (params (MyComp :type Primitive%20Component%20Object%20Reference) (Other :type Actor%20Object%20Reference) (OtherComp :type Primitive%20Component%20Object%20Reference) (bSelfMoved :type Boolean) (HitLocation :type Vector :default "0, 0, 0") (HitNormal :type Vector :default "0, 0, 0") (NormalImpulse :type Vector :default "0, 0, 0") (Hit :type Hit%20Result%20Structure%20%28by%20ref%29))
  ;[1 1 1] If thrown directly at the floor, explode and don't bounce.
  (if (>= (Math|Vector|DotProduct HitNormal "0, 0,1.000000") (Variables|Default|GetZToleranceForFloorExplosion))
    (Game|Components|ProjectileMovement|StopSimulating (Variables|Movement|GetProjectileMovementComponent) Hit)))

(fn ConstructionScript
  (|CallParentFunction)
  (Collision|SetCollisionResponsetoChannel (Variables|Collision|GetCapsuleComponent) "ECC_GameTraceChannel18" "ECR_Overlap"))
